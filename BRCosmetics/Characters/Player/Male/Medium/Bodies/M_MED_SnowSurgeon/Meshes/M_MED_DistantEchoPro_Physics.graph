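# Physics asset for the M_MED_DistantEchoPro male cosmetic body, with capsule bodies for the robe and sleeve dyn bones plus the pelvis, thighs and thigh twist bones.
# Center front and back robe bones constrain to the pelvis, while side flaps constrain to their thighs, with no front-left body.
# Both sleeves also constrain to the pelvis.
# Robe and sleeve joints have limited swing and tight twist, while thigh and thigh-twist joints rotate freely.
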PhysicsAsset /BRCosmetics/Characters/Player/Male/Medium/Bodies/M_MED_SnowSurgeon/Meshes/M_MED_DistantEchoPro_Physics
bodies {
 dyn_robe_bk [{shapeType=Capsule, length=17.4811, radius=6.5, shapeName=dyn_robe_bk_capsule, center=<3.7602 -5 0>}]
 dyn_robe_bk_l [{shapeType=Capsule, length=12, radius=6.5, shapeName=dyn_robe_bk_l_capsule, center=<3.7602 -5 0>, rotation=<rot 90 0 0>}]
 dyn_robe_bk_r [{shapeType=Capsule, length=12, radius=6.5, shapeName=dyn_robe_bk_r_capsule, center=<3.7602 -5 0>, rotation=<rot 90 0 0>}]
 dyn_robe_fr [{shapeType=Capsule, length=12, radius=6, shapeName=dyn_robe_fr_capsule, center=<3.7602 0 0>}]
 dyn_robe_fr_r [{shapeType=Capsule, length=12, radius=6, shapeName=dyn_robe_fr_r_capsule, center=<3.7602 0 0>}]
 dyn_sleeve_l [{shapeType=Capsule, length=10, radius=5, shapeName=dyn_sleeve_l_capsule, center=<6.391 0 3.2458>, rotation=<rot 90 0 0>}]
 dyn_sleeve_r [{shapeType=Capsule, length=10, radius=5, shapeName=dyn_sleeve_r_capsule, center=<6.391 0 3.2458>, rotation=<rot 90 0 0>}]
 pelvis [{shapeType=Capsule, length=10.7102, radius=13.1903, center=<-3.7421 0 0>}]
 thigh_l [{shapeType=Capsule, length=50.5652, radius=10.483, shapeName=thigh_l_capsule, center=<-25.565 -1.5846 0>, rotation=<rot 90 0 0>}]
 thigh_r [{shapeType=Capsule, length=44.8815, radius=11.2659, shapeName=thigh_r_capsule, center=<22.6889 0.7691 0>, rotation=<rot 90 0 0>}]
 thigh_twist_01_l [{shapeType=Capsule, length=10, radius=5, shapeName=thigh_twist_01_l_capsule, rotation=<rot 90 0 0>}]
 thigh_twist_01_r [{shapeType=Capsule, length=10, radius=5, shapeName=thigh_twist_01_r_capsule, rotation=<rot 90 0 0>}]
}
constraints [
 {bone1Name=dyn_robe_fr, bone2Name=pelvis, swing1LimitDegrees=20, swing2LimitDegrees=20, twistLimitDegrees=2}
 {bone1Name=dyn_robe_bk, bone2Name=pelvis, swing1LimitDegrees=25, swing2LimitDegrees=25, twistLimitDegrees=1}
 {bone1Name=dyn_robe_fr_r, bone2Name=thigh_r, swing1LimitDegrees=20, swing2LimitDegrees=20, twistLimitDegrees=2}
 {bone1Name=dyn_robe_bk_l, bone2Name=thigh_l, swing1LimitDegrees=20, swing2LimitDegrees=20, twistLimitDegrees=2}
 {bone1Name=dyn_robe_bk_r, bone2Name=thigh_r, swing1LimitDegrees=25, swing2LimitDegrees=25, twistLimitDegrees=1}
 {
  bone1Name=thigh_l
  bone2Name=pelvis
  swing1LimitDegrees=45
  swing1Motion=Free
  swing2LimitDegrees=45
  swing2Motion=Free
  twistLimitDegrees=45
  twistMotion=Free
 }
 {
  bone1Name=thigh_r
  bone2Name=pelvis
  swing1LimitDegrees=45
  swing1Motion=Free
  swing2LimitDegrees=45
  swing2Motion=Free
  twistLimitDegrees=45
  twistMotion=Free
 }
 {
  bone1Name=thigh_twist_01_l
  bone2Name=thigh_l
  swing1LimitDegrees=45
  swing1Motion=Free
  swing2LimitDegrees=45
  swing2Motion=Free
  twistLimitDegrees=45
  twistMotion=Free
 }
 {
  bone1Name=thigh_twist_01_r
  bone2Name=thigh_r
  swing1LimitDegrees=45
  swing1Motion=Free
  swing2LimitDegrees=45
  swing2Motion=Free
  twistLimitDegrees=45
  twistMotion=Free
 }
 {bone1Name=dyn_sleeve_r, bone2Name=pelvis, swing1LimitDegrees=45, swing2LimitDegrees=20, twistLimitDegrees=2}
 {bone1Name=dyn_sleeve_l, bone2Name=pelvis, swing1LimitDegrees=45, swing2LimitDegrees=20, twistLimitDegrees=2}
]
